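# Lit ammo belt material for Niagara mesh particles and sprites, with scrolling surface maps and procedural waving motion.
# Tinted diffuse, normal and metallic/roughness samples share a panner that takes particle color red as time.
# A height-masked, time-scrolled world-aligned distortion map, driven by dynamic parameters, builds a world-position-offset wave.
# UseSineWaveWPO blends that wave toward a masked directional sine wave with a per-particle time offset.
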Material /BRCosmetics/Weapons/Ammo/Materials/M_AmmoBelt
graph {
 properties {bUsedWithNiagaraMeshParticles=true, bUsedWithNiagaraSprites=true, pixelDepthOffsetMode=PDOM_Legacy, usedShadingModels=Default Lit}
 nodes [
  n1 = Constant3Vector() {constant=<rgba 0 0 1 0>}
  n2 = DynamicParameter() {defaultValue=<rgba 0 0.5 0.05 1>, paramNames [HightOffset, WaveIntensity, MaskOffset, WPO_TextureSize]}
  n3 = WorldPosition()
  n4 = Subtract(A=@n3.XYZ, B=ParticlePositionWS())
  n5 = ComponentMask(None=@n4) {g=true, r=true}
  n6 = Divide(A=ComponentMask(None=@n3.XYZ) {b=true}, B=ScalarParameter() {defaultValue=-4500, parameterName=WaveSpeed}) {constB=5000}
  n7 = Time()
  n8 = Add(A=ComponentMask(None=@n4) {b=true}, B=Multiply(A=Add(A=@n6, B=@n7)) {constB=-190})
  n9 = DynamicParameter() {defaultValue=<rgba 0 0.5 0.05 1>, paramNames [HightOffset, WaveIntensity, MaskOffset, WPO_TextureSize]}
  n10 = AppendVector(A=Constant2Vector(), B=LinearInterpolate(Alpha=@n9.HightOffset) {constB=100})
  n11 = MaterialFunctionCall(
   TextureObject=TextureObjectParameter() {parameterName=WPO_Map, samplerType=SAMPLERTYPE_Normal, texture=/Game/Effects/Fort_Effects/Textures/T_DustDistortion}
   TextureSize=LinearInterpolate(B=Constant() {r=150}, Alpha=@n2.WPO_TextureSize) {constB=150}
   WorldPosition=Add(A=AppendVector(A=@n5, B=@n8), B=@n10)
  ) {materialFunction=&WorldAlignedTexture}
  n12 = WorldPosition()
  n13 = ParticlePositionWS()
  n14 = Divide(A=ComponentMask(None=Subtract(A=@n12.XYZ, B=@n13)) {b=true}) {constB=150}
  n15 = Saturate(None=Multiply(A=Add(A=@n14, B=@n2.MaskOffset) {constB=0.05}) {constB=2})
  n16 = DynamicParameter() {defaultValue=<rgba 0 0.5 0.05 1>, paramNames [HightOffset, WaveIntensity, MaskOffset, WPO_TextureSize]}
  n17 = Multiply(A=@n16.WaveIntensity, B=ScalarParameter() {defaultValue=1, parameterName=WaveIntensity})
  Add(A=Multiply(A=LinearInterpolate(A=@n1, B=@n11.XY Texture, Alpha=@n15), B=LinearInterpolate(Alpha=@n17) {constB=20}) {constB=10}, B=@n10)
  n18 = ScalarParameter() {defaultValue=150, parameterName=WPO_TextureSize}
  n19 = DynamicParameter() {defaultValue=<rgba 0 0.5 0.05 1>, paramNames [HightOffset, WaveIntensity, MaskOffset, WPO_TextureSize]}
  n20 = WorldPosition()
  n21 = Subtract(A=@n20.XYZ, B=ParticlePositionWS())
  n22 = Add(A=Divide(A=ComponentMask(None=@n20.XYZ) {b=true}, B=ScalarParameter() {defaultValue=-4500, parameterName=WaveSpeed}) {constB=5000}, B=Time())
  n23 = AppendVector(A=ComponentMask(None=@n21) {g=true, r=true}, B=Add(A=ComponentMask(None=@n21) {b=true}, B=Multiply(A=@n22) {constB=-190}))
  n24 = DynamicParameter() {defaultValue=<rgba 0 0.5 0.05 1>, paramNames [HightOffset, WaveIntensity, MaskOffset, WPO_TextureSize]}
  n25 = ScalarParameter() {parameterName=WPO_HeightOffset}
  n26 = AppendVector(A=Constant2Vector(), B=LinearInterpolate(Alpha=Add(A=@n24.HightOffset, B=@n25)) {constB=100})
  n27 = Subtract(A=WorldPosition().XYZ, B=ParticlePositionWS())
  n28 = Add(
   A=Divide(A=ComponentMask(None=@n27) {b=true}, B=ScalarParameter() {defaultValue=150, parameterName=WPOMask_Size}) {constB=150}
   B=Add(A=@n19.MaskOffset, B=ScalarParameter() {parameterName=WPOOffset})
  ) {constB=0.05}
  n29 = LinearInterpolate(
   A=Constant3Vector() {constant=<rgba 0 0 1 0>}
   B=MaterialFunctionCall(
    TextureObject=TextureObjectParameter() {parameterName=WPO_Map, samplerType=SAMPLERTYPE_Normal, texture=&T_DustDistortion}
    TextureSize=LinearInterpolate(B=@n18, Alpha=@n19.WPO_TextureSize) {constB=150}
    WorldPosition=Add(A=@n23, B=@n26)
   ) {materialFunction=&WorldAlignedTexture}.XY Texture
   Alpha=Saturate(None=Multiply(A=@n28) {constB=2})
  )
  n30 = DynamicParameter() {defaultValue=<rgba 0 0.5 0.05 1>, paramNames [HightOffset, WaveIntensity, MaskOffset, WPO_TextureSize]}
  n31 = LinearInterpolate(Alpha=Multiply(A=@n30.WaveIntensity, B=ScalarParameter() {defaultValue=1, parameterName=WaveIntensity})) {constB=20}
  n32 = Subtract(A=WorldPosition().XYZ, B=Subtract(A=ActorPositionWS(), B=ParticlePositionWS()))
  n33 = Add(
   A=Divide(A=ComponentMask(None=@n32) {b=true}, B=ScalarParameter() {defaultValue=150, parameterName=WPOMask_Size}) {constB=150}
   B=ScalarParameter() {parameterName=WPOOffset}
  ) {constB=0.05}
  n34 = WorldPosition()
  n35 = Add(
   A=Add(A=Time(), B=DynamicParameter() {parameterIndex=1, defaultValue=<rgba 0 0.5 0.05 1>, paramNames [Time, TimeOffset, "", ""]}.TimeOffset)
   B=Divide(A=ComponentMask(None=@n34.XYZ) {b=true}, B=ScalarParameter() {parameterName=WPOSineSize})
  )
  n36 = ScalarParameter() {defaultValue=1, parameterName=SineWaveSpeed}
  n37 = Multiply(A=Sine(None=Multiply(A=@n35, B=@n36)), B=VectorParameter() {parameterName=SineWaveDirection}.RGB)
  n38 = ScalarParameter() {parameterName=WPOSineAmount}
  LinearInterpolate(
   A=Add(A=Multiply(A=@n29, B=@n31) {constB=10}, B=@n26)
   B=Multiply(A=Saturate(None=@n33), B=Multiply(A=@n37, B=@n38))
   Alpha=ScalarParameter() {parameterName=UseSineWaveWPO}
  )
  n39 = Panner(Time=ParticleColor().R, Speed=AppendVector(A=Constant(), B=ScalarParameter() {parameterName=PanSpeed}))
  n40 = TextureSampleParameter2D(UVs=@n39) {parameterName=SpecularMap, samplerType=SAMPLERTYPE_Masks, texture=/BRCosmetics/Weapons/Ammo/Textures/T_AmmoSet02_S}
  Multiply(A=ScalarParameter() {defaultValue=1, parameterName=MetallicMult}, B=@n40.G)
  Multiply(A=@n40.B, B=ScalarParameter() {defaultValue=1, parameterName=RoughnessMult})
  Multiply(
   A=VectorParameter() {parameterName=BaseColorTint, defaultValue=<rgba 1 1 1 1>}.RGB
   B=TextureSampleParameter2D(UVs=@n39) {parameterName=DiffuseMap, texture=/BRCosmetics/Weapons/Ammo/Textures/T_AmmoSet02_D}.RGB
  )
  Multiply(A=@n33) {constB=2}
  TextureSampleParameter2D(UVs=@n39) {parameterName=NormalMap, samplerType=SAMPLERTYPE_Normal, texture=/BRCosmetics/Weapons/Ammo/Textures/T_AmmoSet02_N}
 ]
}
